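; BP_S23_Part3_RobotTPose is a Season 23 narrative quest actor that adds a scale-based fade-in to its post-interaction visuals.
; On the post-interaction visuals event, it calls the parent, then records the current world scales of the holo cone and skeletal mesh as targets.
; A timeline then lerps both components from preset initial scales to those targets.
; Construction defers to the parent, and the other events do nothing.
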(event VisualState|Meshes|EventPostInteractionVisuals
  (params (PlayerController :type Fort%20Player%20Controller%20Object%20Reference))
  (|CallParentFunction PlayerController)
  ;[1 1 1] Fade-in added by Bruno Afonseca
  (Variables|Default|SetPostInteractionVisualConeFinalScale (Transformation|GetWorldScale (Variables|Default|GetFX%20Holo_Cone)))
  (Variables|Default|SetPostInteractionVisualDeviceFinalScale (Transformation|GetWorldScale (Variables|Default|GetSkeletalMesh)))
  (bind (_direction _scalelerpalpha) (Timeline|PostInteractionVisualFadeIn PlayFromStart)
    (:Update
      (Transformation|SetWorldScale3D (Variables|Default|GetFX%20Holo_Cone) (Math|Vector|Lerp(Vector) (Variables|Default|GetPostInteractionVisualConeInitialScale) (Variables|Default|GetPostInteractionVisualConeFinalScale) _scalelerpalpha))
      (Transformation|SetWorldScale3D (Variables|Default|GetSkeletalMesh) (Math|Vector|Lerp(Vector) (Variables|Default|GetPostInteractionVisualDeviceInitialScale) (Variables|Default|GetPostInteractionVisualDeviceFinalScale) _scalelerpalpha)))))

(event EventBeginPlay
  (|CallParentFunction) ; This node is disabled and will not be called.\nDrag off pins to build functionality.
)

(event Collision|EventActorBeginOverlap
  (params (OtherActor :type Actor%20Object%20Reference)))

(event EventTick
  (params (DeltaSeconds :type Float%20%28single-precision%29)))

(fn ConstructionScript
  (|CallParentFunction))
